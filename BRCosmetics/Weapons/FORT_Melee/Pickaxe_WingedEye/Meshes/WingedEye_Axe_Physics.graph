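# Physics asset for the Winged Eye pickaxe, giving capsule bodies and joint limits to the chain bones of its axe mesh.
# A small capsule on chain_base anchors three larger, equally sized capsules on the dyn_chain bones.
# Each dyn_chain body is constrained directly to chain_base, forming a star rather than links in series.
# Twist and secondary swing limits are tight and uniform, while the middle-numbered bone gets a wider primary swing.
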PhysicsAsset /BRCosmetics/Weapons/FORT_Melee/Pickaxe_WingedEye/Meshes/WingedEye_Axe_Physics
bodies {
 chain_base [{shapeType=Capsule, length=0.505, radius=0.505, shapeName=chain_base_capsule, rotation=<rot 0 -90 0>}]
 dyn_chain_01 [{shapeType=Capsule, length=1, radius=1.8, shapeName=dyn_chain_01_capsule, center=<1.7 -0.0073 0.0217>, rotation=<rot 0 -89.9821 -86.2259>}]
 dyn_chain_02 [{shapeType=Capsule, length=1, radius=1.8, shapeName=dyn_chain_02_capsule, center=<1.7 0.0057 0.0383>, rotation=<rot 0 -90.2324 -87.1562>}]
 dyn_chain_03 [{shapeType=Capsule, length=1, radius=1.8, shapeName=dyn_chain_03_capsule, center=<1.7 -0.0165 0.0374>, rotation=<rot 0 -89.5078 -88.8175>}]
}
constraints [
 {bone1Name=dyn_chain_01, bone2Name=chain_base, swing1LimitDegrees=20, swing2LimitDegrees=10, twistLimitDegrees=3}
 {bone1Name=dyn_chain_02, bone2Name=chain_base, swing1LimitDegrees=45, swing2LimitDegrees=10, twistLimitDegrees=3}
 {bone1Name=dyn_chain_03, bone2Name=chain_base, swing1LimitDegrees=20, swing2LimitDegrees=10, twistLimitDegrees=3}
]
